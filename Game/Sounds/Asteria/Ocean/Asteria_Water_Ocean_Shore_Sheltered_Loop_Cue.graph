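# Looping ambient cue for a sheltered ocean shoreline in Asteria, limited to one concurrent instance.
# It mixes a briefly, randomly delayed lake-lapping loop with a quieter small-wave loop, each with its own close-range attenuation.
# The wave layer has its own volume swell.
# Stacked, endlessly looping envelopers then modulate the whole mix with pitch movement, irregular volume flutter and a broad swell, varying the static loops over time.
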SoundCue /Game/Sounds/Asteria/Ocean/Asteria_Water_Ocean_Shore_Sheltered_Loop_Cue
root=Enveloper_4
nodes [
 {
  id=Attenuation_1
  children [Enveloper_2]
  properties {attenuationSettings=/Game/Sounds/Asteria/Ocean/Attenuations/Asteria_Water_Ocean_Shore_Close_Att}
 }
 {
  id=Attenuation_9
  children [Delay_0]
  properties {attenuationSettings=/Game/Sounds/Asteria/Ocean/Attenuations/Asteria_Water_Ocean_Shore_Very_Close_Att}
 }
 {id=Delay_0, children [WavePlayer_3], properties {delayMax=0.75, delayMin=0.2}}
 {
  id=Enveloper_0
  children [Mixer_0]
  properties {
   bLoop=true
   bLoopIndefinitely=true
   loopEnd=60
   pitchCurve {editorCurveData {keys [
      {interpMode=RCIM_Cubic, time=0}
      <at 15 0.85 RCIM_Cubic>
      {arriveTangent=0.0091, interpMode=RCIM_Cubic, leaveTangent=0.0091, tangentMode=RCTM_User, time=30}
      {interpMode=RCIM_Cubic, time=45}
      {time=60}
     ]}}
   volumeCurve {editorCurveData {keys [<at 0 1>]}}
  }
 }
 {
  id=Enveloper_2
  children [WavePlayer_0]
  properties {
   bLoop=true
   bLoopIndefinitely=true
   loopEnd=20
   pitchCurve {editorCurveData {keys [{time=0}]}}
   volumeCurve {editorCurveData {keys [
      {arriveTangent=-0.01, interpMode=RCIM_Cubic, time=0, value=1}
      {arriveTangent=0.0012, interpMode=RCIM_Cubic, leaveTangent=0.0012, tangentMode=RCTM_User, time=5, value=1.25}
      {arriveTangent=-0.0005, interpMode=RCIM_Cubic, leaveTangent=-0.0005, tangentMode=RCTM_User, time=10, value=0.5}
      {arriveTangent=0.0029, interpMode=RCIM_Cubic, leaveTangent=0.0029, tangentMode=RCTM_User, time=15.0238, value=1.25}
      {interpMode=RCIM_Cubic, leaveTangent=0.0008, time=20, value=1}
     ]}}
  }
 }
 {
  id=Enveloper_3
  children [Enveloper_0]
  properties {
   bLoop=true
   bLoopIndefinitely=true
   loopEnd=60
   pitchCurve {editorCurveData {keys [{time=0}]}}
   volumeCurve {editorCurveData {keys [
      {arriveTangent=-0.01, interpMode=RCIM_Cubic, time=0, value=1}
      {arriveTangent=-0.0035, interpMode=RCIM_Cubic, leaveTangent=-0.0035, tangentMode=RCTM_User, time=4.5847, value=0.8034}
      {arriveTangent=0.0016, interpMode=RCIM_Cubic, leaveTangent=0.0016, tangentMode=RCTM_User, time=7.5626, value=0.9613}
      {arriveTangent=0.0008, interpMode=RCIM_Cubic, leaveTangent=0.0008, tangentMode=RCTM_User, time=11.1496, value=0.8168}
      {arriveTangent=0.0016, interpMode=RCIM_Cubic, leaveTangent=0.0016, tangentMode=RCTM_User, time=13.0193, value=0.8024}
      {arriveTangent=0.0009, interpMode=RCIM_Cubic, leaveTangent=0.0009, time=14.804, value=0.9136}
      {arriveTangent=-0.0011, interpMode=RCIM_Cubic, leaveTangent=-0.0011, tangentMode=RCTM_User, time=18.6287, value=0.8074}
      {arriveTangent=0.0002, interpMode=RCIM_Cubic, leaveTangent=0.0002, time=24.105, value=0.9798}
      {arriveTangent=0.0016, interpMode=RCIM_Cubic, leaveTangent=0.0016, tangentMode=RCTM_User, time=28.0759, value=0.8096}
      {arriveTangent=-0.0003, interpMode=RCIM_Cubic, leaveTangent=-0.0003, time=31.0415, value=0.9283}
      {arriveTangent=0.0009, interpMode=RCIM_Cubic, leaveTangent=0.0009, tangentMode=RCTM_User, time=34.9646, value=0.8077}
      {arriveTangent=-0.0008, interpMode=RCIM_Cubic, leaveTangent=-0.0008, time=38.0791, value=0.9798}
      <at 42.05 0.8023>
      {arriveTangent=0.0024, interpMode=RCIM_Cubic, leaveTangent=0.0024, tangentMode=RCTM_User, time=44.6087, value=0.8074}
      {arriveTangent=-0.0006, interpMode=RCIM_Cubic, leaveTangent=-0.0006, time=48.1071, value=0.9613}
      {arriveTangent=-0.0056, interpMode=RCIM_Cubic, leaveTangent=-0.0056, tangentMode=RCTM_User, time=53.1761, value=0.8024}
      <at 55.1553 0.9384>
      {arriveTangent=-0.0003, interpMode=RCIM_Cubic, leaveTangent=-0.0003, tangentMode=RCTM_User, time=57.7955, value=0.8022}
      <at 60 1 RCIM_Cubic>
     ]}}
  }
 }
 {
  id=Enveloper_4
  children [Enveloper_3]
  properties {
   bLoop=true
   bLoopIndefinitely=true
   loopEnd=20
   pitchCurve {editorCurveData {keys [{time=0}]}}
   volumeCurve {editorCurveData {keys [
      {arriveTangent=-0.01, interpMode=RCIM_Cubic, time=0, value=1}
      <at 10 1.2 RCIM_Cubic>
      {interpMode=RCIM_Cubic, leaveTangent=0.01, time=20, value=1}
     ]}}
  }
 }
 {id=Mixer_0, children [Attenuation_9, Attenuation_1], properties {inputVolume [1.2,0.8]}}
 {id=WavePlayer_0, properties {bLooping=true, soundWaveAssetPtr=/Game/Athena/Sounds/Water/Ambient/Amb_Water_Ocean_Shore_SmWaves_Loop_01}}
 {id=WavePlayer_3, properties {bLooping=true, soundWaveAssetPtr=/Game/Athena/Sounds/Ambient/Water_Lake_Shore_Lapping_02}}
]
properties {
 attenuationSettings=/Game/Athena/Sounds/Ambient/Fort_Amb_Water_River_Att
 bOverrideConcurrency=true
 maxDistance=8600
 soundClassObject=&Fort_Ambient_Outdoors_Athena
 volumeMultiplier=0.802
 attenuationOverrides {falloffDistance=2000, nonSpatializedRadiusStart=700, stereoSpread=1000, attenuationShapeExtents {x=925}}
 concurrencyOverrides {maxCount=1, voiceStealReleaseTime=3}
}
